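# Sound cue for the finish of a capture on the Creative capture device toy, playing a single wave directly from one wave player node.
# It routes through the critical in-game UI sound class at slightly reduced volume.
# It references the Athena trap activation attenuation but overrides it with non-spatialized settings and a shorter falloff distance.
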SoundCue /Game/Sounds/Creative/Toys/CaptureDevice/CaptureDevice_Capture_Finish_Cue
root=WavePlayer_0
nodes [{id=WavePlayer_0, properties {soundWaveAssetPtr=&/CaptureDevice_Capture_Finish_01}}]
properties {
 attenuationSettings=/Game/Sounds/Attenuation/Fort_Traps_Activate_Athena_Att
 bOverrideAttenuation=true
 maxDistance=1600
 soundClassObject=&Fort_UI_InGame_Critical
 volumeMultiplier=0.7
 attenuationOverrides {bSpatialize=false, falloffDistance=1200}
}
