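# Sound cue for the looping capture sound of the Creative Capture Device toy, in the Creative gadgets sound class.
# A looping power-up wave player feeds a continuous modulator, which lets a runtime "Pitch" parameter vary the pitch within a narrow range.
# The cue plays at reduced volume and raised base pitch.
# Its attenuation override disables spatialization and uses a small radius with zero falloff distance.
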SoundCue /Game/Sounds/Creative/Toys/CaptureDevice/CaptureDevice_Capture_Loop_Cue
root=ModulatorContinuous_0
nodes [
 {
  id=ModulatorContinuous_0
  children [WavePlayer_0]
  properties {pitchModulationParams {maxInput=1.2, maxOutput=1.2, minInput=0.9, minOutput=0.9, parameterName=Pitch}}
 }
 {id=WavePlayer_0, properties {bLooping=true, soundWaveAssetPtr=&/CaptureArea_04_PowerUp_Loop}}
]
properties {
 attenuationSettings=/Game/Sounds/Attenuation/Fort_Traps_Activate_Athena_Att
 bOverrideAttenuation=true
 maxDistance=200
 pitchMultiplier=1.2
 soundClassObject=&Fort_Gadgets_Creative
 volumeMultiplier=0.5
 attenuationOverrides {bSpatialize=false, falloffDistance=0, attenuationShapeExtents {x=200}}
}
