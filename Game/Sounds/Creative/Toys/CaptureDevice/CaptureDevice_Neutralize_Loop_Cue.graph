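# Looping sound cue for the neutralize state of the Creative capture device, classed as a Creative gadget sound.
# A looping power-down wave feeds a continuous modulator, which passes a runtime "Pitch" parameter through one-to-one within a narrow range.
# The cue lowers base pitch and volume and overrides attenuation with non-spatialized, distance-based falloff.
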SoundCue /Game/Sounds/Creative/Toys/CaptureDevice/CaptureDevice_Neutralize_Loop_Cue
root=ModulatorContinuous_0
nodes [
 {
  id=ModulatorContinuous_0
  children [WavePlayer_2]
  properties {pitchModulationParams {maxInput=1.2, maxOutput=1.2, minInput=0.9, minOutput=0.9, parameterName=Pitch}}
 }
 {id=WavePlayer_2, properties {bLooping=true, soundWaveAssetPtr=&/CaptureArea_04_PowerDown_Loop}}
]
properties {
 attenuationSettings=/Game/Sounds/Attenuation/Fort_Traps_Activate_Athena_Att
 bOverrideAttenuation=true
 maxDistance=1600
 pitchMultiplier=0.8
 soundClassObject=&Fort_Gadgets_Creative
 volumeMultiplier=0.5
 attenuationOverrides {bSpatialize=false, falloffDistance=1200}
}
